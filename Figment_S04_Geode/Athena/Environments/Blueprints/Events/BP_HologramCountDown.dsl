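; Holotable countdown hologram for the Geode event, which shows or hides its mesh and plays countdown audio.
; On authority it binds to BP_GeodeScripting's countdown-finished event to clear visibility, and a calendar event-over flag also triggers cleanup.
; Remotes bind to timer updates for tick cues, a faded-in warning in the last sixty seconds and a one-shot launch cue at zero.
; OnRep functions drive visibility, loop audio and countdown material phase.
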(event Custom|CleanUpOnSequenceFinished
  (Variables|Default|SetIsVisible false))

(event EventBeginPlay
  (Utilities|FlowControl|Sequence
    (:then_0
      ;[1 1 1] Spawn Audio
      (bind _returnvalue (Audio|SpawnSoundAttached :Sound "/Figment_S04_Geode/Sounds/Fort_Gadgets/Holotable/Holotable_Loop_Cue.Holotable_Loop_Cue" :AttachToComponent (Variables|Default|GetSM_CountdownHologram) :AttachPointName "None" :LocationType "KeepRelativeOffset" :bStopWhenAttachedToDestroyed true :VolumeMultiplier 1.0 :PitchMultiplier 1.0 :StartTime 0.0 :bAutoDestroy true))
      (Variables|Default|SetHolotableLoopComp _returnvalue)
      (Networking|SwitchHasAuthority
        (:Authority
          (bind _outactors (Actor|GetAllActorsOfClass "/Figment_S04_Geode/Athena/Maps/Test/Events/BP_GeodeScripting.BP_GeodeScripting_C"))
          (Default|BindEventtoCountdownFinished30s (Utilities|Array|Get(acopy) _outactors) (delegate Custom|CleanUpOnSequenceFinished))
          (Variables|Default|SetIsVisible true))
        (:Remote
          (bind _outactors_2 (Actor|GetAllActorsOfClass "/Figment_S04_Geode/Athena/Maps/Test/Events/BP_GeodeScripting.BP_GeodeScripting_C"))
          (Default|BindEventtoUpdateCountdownTimer (Utilities|Array|Get(acopy) _outactors_2) (delegate Custom|OnUpdateCountdownTimer)))))
    (:then_1
      (MCPContext|WatchCalendarEvent (Variables|Default|GetEventOverFlag)
        (:OnEventActive
          (CallFunction|CleanUponSequenceFinished))))))

(event Custom|OnUpdateCountdownTimer
  (params (SecondsRemaining :type Integer))
  (Audio|PlaySoundatLocation (select (<= SecondsRemaining 60) "/Figment_S04_Geode/Sounds/Fort_Gadgets/Holotable/Holotable_Countdown_Final_Cue.Holotable_Countdown_Final_Cue" "/Figment_S04_Geode/Sounds/Fort_Gadgets/Holotable/Holotable_Countdown_Cue.Holotable_Countdown_Cue") (Transformation|GetWorldLocation (Variables|Default|GetSM_CountdownHologram)))
  (CallFunction|HandleCountdownUpdate SecondsRemaining))

(event EventEndPlay
  (params (EndPlayReason :type EEndPlayReason%20Enum :default "Destroyed"))
  (bind _holotableloopcomp (Variables|Default|GetHolotableLoopComp)
    (:then
      (Audio|Components|Audio|Stop _holotableloopcomp))
    (:else)))

(event EventDestroyed
  (bind _holotableloopcomp (Variables|Default|GetHolotableLoopComp)
    (:then
      (Audio|Components|Audio|Stop _holotableloopcomp))
    (:else)))

(event Collision|EventActorBeginOverlap
  (params (OtherActor :type Actor%20Object%20Reference)))

(event Custom|UpdateCountDown)

(event EventTick
  (params (DeltaSeconds :type Float%20%28single-precision%29)))

(fn HandleCountdownUpdate
  (params (Seconds :type Integer))
  (Utilities|FlowControl|Sequence
    (:then_0
      (if (> Seconds 3)
        (Variables|Default|SetbPlayedFinalWarning false)))
    (:then_1
      (if (> Seconds 60)
        (Variables|Default|GetLaunchSixtySecondsWarning
          (:then
            (Audio|Components|Audio|FadeOut (Variables|Default|GetLaunchSixtySecondsWarning) 0.2 0.0 "Linear" :!class /Script/Engine.AudioComponent))
          (:else))
        (else
          (Variables|Default|GetLaunchSixtySecondsWarning
            (:then
              (Variables|Default|GetLaunchSixtySecondsWarning
                (:then
                  (if (and (<= Seconds 0) (not (Variables|Default|GetbPlayedFinalWarning)))
                    (Audio|Components|Audio|FadeOut (Variables|Default|GetLaunchSixtySecondsWarning) 1.0 0.0 "Linear" :!class /Script/Engine.AudioComponent)
                    (Audio|PlaySoundatLocation "/Figment_S04_Geode/Sounds/Fort_Gadgets/Holotable/Holotable_Launch_Cue.Holotable_Launch_Cue" (Transformation|GetWorldLocation (Variables|Default|GetSM_CountdownHologram)))
                    (Variables|Default|SetbPlayedFinalWarning true)))
                (:else)))
            (:else
              (if (and (<= Seconds 60) (> Seconds 0))
                (bind _returnvalue (Audio|SpawnSoundAttached "/Figment_S04_Geode/Sounds/Fort_Gadgets/Holotable/Holotable_LaunchWarning_Cue.Holotable_LaunchWarning_Cue" (Variables|Default|GetSM_CountdownHologram)))
                (bind _output_get (Variables|Default|SetLaunchSixtySecondsWarning _returnvalue))
                (Audio|Components|Audio|FadeIn _output_get 3.0 1.0 0.0 "Linear" :!class /Script/Engine.AudioComponent)))))))))

(fn OnRep_IsAlive
  (Development|SetHiddeninGame (Variables|Default|GetSM_CountdownHologram) (not (Variables|Default|GetIsVisible)) false)
  (if (Variables|Default|GetIsVisible)
    (bind _holotableloopcomp (Variables|Default|GetHolotableLoopComp)
      (:then
        (Audio|Components|Audio|Play _holotableloopcomp 0.0))
      (:else))
    (else
      (bind _holotableloopcomp_2 (Variables|Default|GetHolotableLoopComp)
        (:then
          (Audio|Components|Audio|Stop _holotableloopcomp_2))
        (:else)))))

(fn OnRep_ServerStartingDateTime
  (Variables|Default|SetServerTimeOnStart (GameState|GetServerWorldTimeSeconds (Game|GetGameState)))
  (if (Math|Timespan|Timespan>Timespan (Math|DateTime|DateTime-DateTime (Variables|Default|GetStart) (Variables|Default|GetServerStartingDateTime)) (Math|Timespan|MakeTimespan 0 0 30))
    (Actor|Tick|SetActorTickEnabled false)))

(fn OnRep_ShowPhase2
  (Rendering|Material|SetScalarParameterValue "/Game/Athena/Events/Generic/CountdownTimer/Materials/MPC_Countdown.MPC_Countdown" "ShowPhase2" (Variables|Default|GetShowPhase2) :!class /Script/Engine.KismetMaterialLibrary))

(fn ConstructionScript)
