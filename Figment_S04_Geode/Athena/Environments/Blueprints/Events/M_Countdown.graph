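# Translucent hologram material for the countdown timer event, showing a progress bar, a rocket or countdown image, and timer digits or blinking launch text.
# The second UV set's U picks the mesh region, while MPC_Countdown drives progress, phase two, launch switching and eight digits.
# Nested Ifs map eleven slots into a digit atlas, sampled with explicit derivatives.
# Scanlines modulate opacity, and vertex red masks camera alignment plus wiggle offset.
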Material /Figment_S04_Geode/Athena/Environments/Blueprints/Events/M_Countdown
graph {
 properties {blendMode=BLEND_Translucent, pixelDepthOffsetMode=PDOM_Legacy, usedShadingModels=Default Lit}
 nodes [
  n1 = ComponentMask(None=TextureCoordinate() {coordinateIndex=1}) {g=true}
  n2 = Constant() {r=0.001}
  n3 = Divide(A=@n1, B=@n2) {constB=0.001}
  n4 = CollectionParameter() {collection=/Game/Athena/Events/Generic/CountdownTimer/Materials/MPC_Countdown, parameterName=ShowPhase2}
  n5 = Divide(
   A=LinearInterpolate(
    A=Constant()
    B=CollectionParameter() {collection=&MPC_Countdown, parameterName=ProgressNormalized}
    Alpha=@n4
   )
   B=@n2
  ) {constB=0.001}
  Add(A=@n3, B=@n5)
  n6 = MaterialFunctionCall!(
   WPO Intensity A=ScalarParameter() {defaultValue=0.2, parameterName=WiggleIntensity1}
   WPO Intensity B=ScalarParameter() {defaultValue=-0.2, parameterName=WiggleIntensity2}
   Time
   Bias=ScalarParameter() {defaultValue=0.7058, parameterName=Bias}
  ) {materialFunction=/Game/Effects/Fort_Effects/Effects/Atlas/Materials/MF_Atlas_HologramWiggle_01}
  n7 = MaterialFunctionCall!(
   Number of Scanlines=ScalarParameter() {defaultValue=150, parameterName=Scanline Amount}
   Scanline Speed
   "Scanlines Thickness (Bias)"=ScalarParameter() {defaultValue=-0.15, parameterName=ScanlineSize}
   Scanlines Dark Value=ScalarParameter() {defaultValue=0.5, parameterName=Scanlines Dark}
   Scanlines Light Value=ScalarParameter() {defaultValue=1, parameterName=Scanlines Light}
   Block Opacity=StaticSwitchParameter(True=@n6.Intense 0-1, False=Constant()) {defaultValue=true, parameterName=Use Blocks?}
   UVs
   Time
   Block Repeat
  ) {materialFunction=/Game/Effects/Fort_Effects/Effects/Atlas/Materials/MF_Atlas_Scanlines_01}
  n8 = Multiply(A=@n7.Secondary Lines, B=ScalarParameter() {defaultValue=1, parameterName=Secondary Line Opacity})
  n9 = ComponentMask(None=TextureCoordinate() {coordinateIndex=1}) {r=true}
  n10 = Constant()
  n11 = MaterialFunctionCall(
   UVs=TextureCoordinate()
   Texture Scale=AppendVector(A=ScalarParameter() {defaultValue=1, parameterName=u}, B=ScalarParameter() {defaultValue=1, parameterName=v})
  ) {materialFunction=&ScaleUVsByCenter}
  n12 = Multiply(A=ComponentMask(None=@n11.UVs) {r=true}) {constB=11}
  n13 = Constant() {r=10}
  n14 = ComponentMask(None=@n12) {r=true}
  n15 = Constant() {r=7}
  n16 = Constant() {r=4}
  n17 = Constant() {r=1}
  n18 = CollectionParameter() {collection=&MPC_Countdown, parameterName=Digit8}
  n19 = CollectionParameter() {collection=&MPC_Countdown, parameterName=Digit7}
  n20 = If(
   A=Constant() {r=3}
   B=@n14
   "A > B"=If(A=Constant() {r=2}, B=@n14, "A > B"=If(A=@n17, B=@n14, "A > B"=@n18, "A < B"=@n19), "A < B"=Constant() {r=-1})
   "A < B"=CollectionParameter() {collection=&MPC_Countdown, parameterName=Digit6}
  )
  n21 = CollectionParameter() {collection=&MPC_Countdown, parameterName=Digit5}
  n22 = If(
   A=Constant() {r=6}
   B=@n14
   "A > B"=If(A=Constant() {r=5}, B=@n14, "A > B"=If(A=@n16, B=@n14, "A > B"=@n20, "A < B"=@n21), "A < B"=Constant() {r=-1})
   "A < B"=CollectionParameter() {collection=&MPC_Countdown, parameterName=Digit4}
  )
  n23 = CollectionParameter() {collection=&MPC_Countdown, parameterName=Digit3}
  n24 = If(
   A=Constant() {r=9}
   B=@n14
   "A > B"=If(A=Constant() {r=8}, B=@n14, "A > B"=If(A=@n15, B=@n14, "A > B"=@n22, "A < B"=@n23), "A < B"=Constant() {r=-1})
   "A < B"=CollectionParameter() {collection=&MPC_Countdown, parameterName=Digit2}
  )
  n25 = CollectionParameter() {collection=&MPC_Countdown, parameterName=Digit1}
  n26 = Add(A=Frac(None=@n12), B=Floor(None=If(A=@n13, B=@n14, "A > B"=@n24, "A < B"=@n25)))
  n27 = TextureCoordinate()
  n28 = TextureSample!(
   UVs=AppendVector(A=Divide(A=@n26) {constB=10}, B=ComponentMask(None=@n11.UVs) {g=true})
   Tex
   "DDX(UVs)"=DDX(Value=@n27)
   "DDY(UVs)"=DDY(Value=@n27)
   Apply View MipBias
  ) {mipValueMode=TMVM_Derivative, samplerType=SAMPLERTYPE_Masks, texture=/Game/Athena/Environments/Blueprints/Events/T_Numbers_M}
  n29 = Multiply(A=Time(), B=ScalarParameter() {defaultValue=2.7835, parameterName=BlinkSpeed})
  n30 = Multiply(
   A=TextureSample!(UVs=@n11.UVs, Tex, Apply View MipBias) {texture=/Game/Athena/Environments/Blueprints/Events/T_LaunchText_M}.RGB
   B=Clamp(None=Sine(None=@n29))
  )
  n31 = CollectionParameter() {collection=&MPC_Countdown, parameterName=Launch}
  n32 = LinearInterpolate(
   A=TextureSample() {texture=&/T_RocketHologram}.RGB
   B=TextureSample() {texture=&/T_CountdownHologram}.RGB
   Alpha=@n4
  )
  n33 = If(
   A=@n9
   B=Constant() {r=1}
   "A > B"=LinearInterpolate(A=Constant(), B=LinearInterpolate(A=@n28.RGB, B=@n30, Alpha=@n31), Alpha=@n4)
   "A < B"=Multiply(A=@n32, B=ScalarParameter() {defaultValue=6.4926, parameterName=Opacity})
  )
  n34 = Clamp(None=Subtract(A=@n3, B=@n5))
  Add(A=@n8, B=Multiply(A=@n7.Result, B=If(A=@n9, B=@n10, "A > B"=@n33, "A < B"=@n34)))
  LinearInterpolate(
   A=VectorParameter() {parameterName=ColorLaunch_1, defaultValue=<rgba 0.026 0.0004 0 1>}.RGB
   B=VectorParameter() {parameterName=ColorLaunch_2, defaultValue=<rgba 1 0 0 1>}.RGB
   Alpha=@n30
  )
  n35 = VectorParameter() {parameterName=Color, defaultValue=<rgba 1 0 0 1>}
  n36 = VectorParameter() {parameterName=Color_2, defaultValue=<rgba 2 0.2383 0.1843 1>}
  n37 = Add(A=ComponentMask(None=@n32) {r=true}, B=@n8)
  MaterialFunctionCall(Emissive=If(
    A=@n9
    B=Constant() {r=1}
    "A > B"=VectorParameter() {parameterName=Color_Text, defaultValue=<rgba 10 0.75 0.75 1>}.RGB
    "A < B"=LinearInterpolate(A=@n35.RGB, B=@n36.RGB, Alpha=@n37)
   )) {materialFunction=&MF_FX_EmissiveAdjustments}
  n38 = Constant3Vector() {constant=<rgba 1 0 0 0>}
  n39 = Constant3Vector() {constant=<rgba 0 0 -1 0>}
  n40 = Constant3Vector() {constant=<rgba 0 1 0 0>}
  Multiply(
   A=Add(
    A=MaterialFunctionCall!(WorldPosition, PivotLocation, WS Normals, Custom Object Basis 1=@n38, Custom Object Basis 2=@n39, Custom Object Basis 3=@n40) {materialFunction=/Engine/Functions/Engine_MaterialFunctions02/WorldPositionOffset/AlignMeshToTheCamera}.World Position Offset
    B=@n6.Result
   )
   B=VertexColor().R
  )
  OneMinus(None=@n1)
  OneMinus(None=@n34)
  ScalarParameter() {parameterName=CountDown}
 ]
}
