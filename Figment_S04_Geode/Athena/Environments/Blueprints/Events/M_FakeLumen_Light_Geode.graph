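# Unlit material for a fake Lumen light in the Geode event environment, producing a tinted color that varies through a time-of-day function.
# The Color parameter is multiplied by itself, then scaled by the result of Scalar_Time_of_Day.
# All four time-of-day inputs read the Morning channel of TimeOfDayAdjustments, so its Day, Evening and Night channels are unused.
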Material /Figment_S04_Geode/Athena/Environments/Blueprints/Events/M_FakeLumen_Light_Geode
graph {
 properties {bAutomaticallySetUsageInEditor=false, shadingModel=MSM_Unlit, usedShadingModels=Unlit}
 nodes [
  n1 = VectorParameter() {parameterName=Color, channelNames {a=Multiplier, b=Blue, g=Green, r=Red}, defaultValue=<rgba 1 0.6442 0.1098 1>}
  n2 = VectorParameter() {parameterName=TimeOfDayAdjustments, channelNames {a=Night, b=Evening, g=Day, r=Morning}, defaultValue=<rgba 1 1 1 1>}
  Multiply(
   A=Multiply(A=@n1, B=@n1)
   B=MaterialFunctionCall(Value_Morning=@n2.Morning, Value_Day=@n2.Morning, Value_Evening=@n2.Morning, Value_Night=@n2.Morning) {materialFunction=&Scalar_Time_of_Day}.Result
  )
 ]
}
